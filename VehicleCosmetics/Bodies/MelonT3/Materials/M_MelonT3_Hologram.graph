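# Translucent, unlit, two-sided hologram material for the MelonT3 vehicle cosmetic body, usable on skeletal meshes.
# Emissive brightens a hologram texture with panning UV and screen-space noise and scanlines, masked by vertex red and tinted by CustomColor.
# A time-and-U sine pulse blends the noise patterns and joins texture alpha, vertex green and local-space reflection vector to form opacity.
# Sine/cosine world position offset adds wobble, phase-shifted by vertex alpha.
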Material /VehicleCosmetics/Bodies/MelonT3/Materials/M_MelonT3_Hologram
graph {
 properties {
  bAutomaticallySetUsageInEditor=false
  bUsedWithSkeletalMesh=true
  blendMode=BLEND_Translucent
  shadingModel=MSM_Unlit
  twoSided=true
  usedShadingModels=Unlit
 }
 nodes [
  n1 = Add(A=Time(), B=VertexColor().A)
  n2 = AppendVector(A=AppendVector(A=Sine(None=@n1) {period=2}, B=Cosine(None=@n1) {period=2}), B=Sine(None=@n1) {period=1.333})
  NamedRerouteDeclaration(None=Multiply(
    A=Transform(None=@n2) {transformSourceType=TRANSFORMSOURCE_World, transformType=TRANSFORM_Local}
    B=Constant3Vector() {constant=<rgba 0.1 0.1 0.1 1>}
   )) {name=WPO, nodeColor=<rgba 1 0 0.9141 1>}
  n3 = TextureSampleParameter2D() {
   parameterName=Melon_Holograms_00
   samplerSource=SSM_Wrap_WorldGroupSettings
   texture=/VehicleCosmetics/Bodies/MelonT1/Textures/T_Melon_Holograms
  }
  n4 = Clamp(None=ConstantBiasScale(None=@n3.B) {bias=0.05, scale=5})
  n5 = Multiply(A=Time(), B=Constant() {r=0.15})
  n6 = TextureSample(UVs=Panner(Time=@n5) {speedY=0.25}) {samplerSource=SSM_Wrap_WorldGroupSettings, texture=/VehicleCosmetics/Wheels/_SharedTextures/FX/Noise_Generic_01_Pack}
  n7 = Divide(A=ComponentMask(None=ScreenPosition().ViewportUV) {g=true, r=true}, B=Constant() {r=0.5})
  n8 = TextureSample(UVs=Panner(Coordinate=@n7, Time=@n5) {speedY=-0.1}) {samplerSource=SSM_Wrap_WorldGroupSettings, texture=&Noise_Generic_01_Pack}
  n9 = Add(A=Multiply(A=@n6.R, B=@n8.G), B=Multiply(A=@n6.G, B=@n8.R))
  n10 = TextureSample(UVs=Panner(Coordinate=@n7, Time=@n5) {speedY=0.05}) {samplerSource=SSM_Wrap_WorldGroupSettings, texture=&Noise_Generic_01_Pack}
  n11 = TextureSample(UVs=Panner(Coordinate=TextureCoordinate() {vTiling=16}) {speedY=0.2}) {samplerSource=SSM_Wrap_WorldGroupSettings, texture=/VehicleCosmetics/Wheels/_SharedTextures/FX/Scan_M}
  n12 = Add(A=Multiply(A=Multiply(A=@n10.R, B=@n9)) {constB=16}, B=LinearInterpolate(Alpha=Multiply(A=@n11.R, B=@n8.R)) {constA=0.13, constB=0.5})
  n13 = VertexColor()
  n14 = Multiply(A=@n13.R, B=@n13.R)
  n15 = Add(A=@n10.R, B=@n8.R)
  n16 = Multiply(A=Multiply(A=Add(A=@n6.B, B=@n15)) {constB=0.5}, B=Add(A=Multiply(A=@n10.R, B=@n6.G), B=Multiply(A=@n6.R, B=@n10.G)))
  n17 = Multiply(A=@n11.G, B=Add(A=@n10.G, B=@n8.G))
  n18 = Multiply(A=@n14, B=Add(A=Multiply(A=@n16) {constB=16}, B=LinearInterpolate(Alpha=@n17) {constA=0.13, constB=0.5}))
  n19 = Add(A=Multiply(A=Time()) {constB=0.15}, B=ComponentMask(None=TextureCoordinate()) {r=true})
  n20 = Power(Base=Abs(None=Sine(None=@n19) {period=3})) {constExponent=0.15}
  n21 = LinearInterpolate(A=@n4, B=Multiply(A=@n4) {constB=3}, Alpha=Multiply(A=LinearInterpolate(A=@n12, B=@n18, Alpha=@n20), B=@n14))
  NamedRerouteDeclaration(None=Multiply(A=Multiply(A=@n21) {constB=2.5}, B=VectorParameter() {parameterName=CustomColor, defaultValue=<rgba 0.2582 0.8796 0.1356 1>}.RGB)) {name=Emissive, nodeColor=<rgba 1 0.4453 0 1>}
  n22 = Transform(None=ReflectionVectorWS()) {transformSourceType=TRANSFORMSOURCE_World, transformType=TRANSFORM_Local}
  n23 = Add(A=Multiply(A=Multiply(A=@n20, B=@n3.A), B=@n13.G), B=Multiply(A=Multiply(A=@n22) {constB=0.5}, B=@n3.A))
  NamedRerouteDeclaration(None=Clamp(None=Multiply(A=@n23) {constB=0.75})) {name=Opacity, nodeColor=<rgba 1 0 0.5156 1>}
  NamedRerouteUsage()
  NamedRerouteUsage()
  NamedRerouteUsage()
 ]
}
